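# Parent material for Babymaker vehicle wheels, adding an animated, color-tinted glow to the shared simple wheel material function.
# Two slowly rotating noise samples with different centers are summed, contrast-adjusted, and multiplied by custom color, emission intensity and Fresnel.
# The tire mask's red channel and vertex color red confine this effect, which drives base color and emissive.
# The result passes through vehicle damage and de-rez transition functions.
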Material /VehicleCosmetics/Wheels/Babymaker/Materials/M_WheelParent_Babymaker
graph {
 properties {
  bAutomaticallySetUsageInEditor=false
  bUseMaterialAttributes=true
  bUsedWithMeshDeformer=true
  bUsedWithSkeletalMesh=true
  pixelDepthOffsetMode=PDOM_Legacy
  usedShadingModels=Default Lit
 }
 nodes [
  n1 = MaterialFunctionCall() {materialFunction=&MF_WheelParent_Simple}
  n2 = GetMaterialAttributes(None=@n1.MaterialAttributes) {attributeGetTypes [#1]}
  n3 = ComponentMask(None=@n1.TireRGBAMask) {r=true}
  n4 = Add(
   A=TextureSample(UVs=Rotator() {centerX=1, centerY=1, speed=0.05}) {samplerSource=SSM_Wrap_WorldGroupSettings, texture=/VehicleCosmetics/Wheels/_SharedTextures/FX/ContrastNoiseGradient_T}.B
   B=TextureSample(UVs=Rotator() {centerX=-1, centerY=-2, speed=0.05}) {samplerSource=SSM_Wrap_WorldGroupSettings, texture=&ContrastNoiseGradient_T}.R
  )
  n5 = Multiply(
   A=Clamp(None=ConstantBiasScale(None=@n4) {bias=-0.15, scale=3}) {minDefault=0.65}
   B=VectorParameter() {group=5_Special Effects, parameterName=Custom Color, defaultValue=<rgba 0.1912 0.6795 0.9387 1>}.RGB
  )
  n6 = ScalarParameter() {defaultValue=1, group=5_Special Effects, parameterName=Emission Intensity}
  n7 = Constant() {r=1}
  n8 = Multiply(
   A=@n3
   B=Multiply(
    A=Multiply(A=@n5, B=@n6)
    B=MaterialFunctionCall!(Normal, Exponent=@n7) {materialFunction=/VehicleCosmetics/Wheels/Cola_Holographic/Materials/MF_Fresnel_Classic}.Result
   )
  )
  n9 = LinearInterpolate(A=@n2.Base Color, B=LinearInterpolate(A=@n8, B=@n2.Base Color, Alpha=OneMinus(None=@n3)), Alpha=@n1.VertexColorR)
  n10 = Multiply(A=@n1.VertexColorR, B=@n8)
  MaterialFunctionCall(In=MaterialFunctionCall(Material=SetMaterialAttributes!(MaterialAttributes=@n2.MaterialAttributes, Base Color=@n9, Emissive Color=@n10) {attributeSetTypes [#1, #2]}) {materialFunction=&MF_BaseVehicle_Damage}.Result) {materialFunction=&MF_VehicleDeRez_Transition}
 ]
}
