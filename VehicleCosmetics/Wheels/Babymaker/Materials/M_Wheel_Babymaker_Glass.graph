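# Translucent, two-sided, surface-lit glass material for the Babymaker vehicle wheel, enabled for skeletal meshes and mesh deformers.
# Two channels of a shared noise gradient, slowly rotated about different pivots, are summed, contrast-boosted and clamped to a raised floor for an animated shimmer.
# Custom Color and Emission Intensity tint and scale that shimmer, which a borrowed classic Fresnel function then modulates.
# Scalar parameters expose Metallic, Roughness and Opacity.
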Material /VehicleCosmetics/Wheels/Babymaker/Materials/M_Wheel_Babymaker_Glass
graph {
 properties {
  bAutomaticallySetUsageInEditor=false
  bUsedWithMeshDeformer=true
  bUsedWithSkeletalMesh=true
  blendMode=BLEND_Translucent
  previewMesh=/VehicleCosmetics/Wheels/RL131/Meshes/SK_Wheel_RL131_Front
  translucencyLightingMode=TLM_Surface
  twoSided=true
  usedShadingModels=Default Lit
 }
 nodes [
  LinearInterpolate(A=ScalarParameter() {defaultValue=1, group=00 Base, parameterName=Metallic})
  LinearInterpolate(A=ScalarParameter() {defaultValue=0.2, group=00 Base, parameterName=Roughness}) {constB=0}
  Reroute(None=ScalarParameter() {defaultValue=0.2, group=00 Base, parameterName=Opacity})
  n1 = Add(
   A=TextureSample(UVs=Rotator() {centerX=1, centerY=1, speed=0.05}) {samplerSource=SSM_Wrap_WorldGroupSettings, texture=/VehicleCosmetics/Wheels/_SharedTextures/FX/ContrastNoiseGradient_T}.B
   B=TextureSample(UVs=Rotator() {centerX=-1, centerY=-2, speed=0.05}) {samplerSource=SSM_Wrap_WorldGroupSettings, texture=&ContrastNoiseGradient_T}.R
  )
  n2 = Multiply(
   A=Clamp(None=ConstantBiasScale(None=@n1) {bias=-0.15, scale=3}) {minDefault=0.65}
   B=VectorParameter() {group=00 Base, parameterName=Custom Color, defaultValue=<rgba 0.1912 0.6795 0.9387 1>}.RGB
  )
  n3 = ScalarParameter() {defaultValue=1, group=00 Base, parameterName=Emission Intensity}
  n4 = ScalarParameter() {defaultValue=1, parameterName=Fresnel_int}
  Reroute(None=Multiply(
    A=Multiply(A=@n2, B=@n3)
    B=MaterialFunctionCall!(Normal, Exponent=@n4) {materialFunction=/VehicleCosmetics/Wheels/Cola_Holographic/Materials/MF_Fresnel_Classic}.Result
   ))
 ]
}
